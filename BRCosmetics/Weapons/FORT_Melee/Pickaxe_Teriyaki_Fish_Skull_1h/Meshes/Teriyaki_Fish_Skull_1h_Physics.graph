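# Physics asset for the Teriyaki Fish Skull 1h pickaxe, giving its fish skeleton mesh simulated bodies for secondary motion.
# Box bodies form a spine chain from the weapon handle through four segments, with the head and both fins constrained to the last one.
# Narrow angular limits keep spine and head nearly rigid, while fins and tail swing wider.
# The tail constrains only to its base body, which, like the spine base body, has no constraint to the chain.
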PhysicsAsset /BRCosmetics/Weapons/FORT_Melee/Pickaxe_Teriyaki_Fish_Skull_1h/Meshes/Teriyaki_Fish_Skull_1h_Physics
bodies {
 Melee_TwoHanded_Handle [{shapeType=Box, extentX=0.5, extentY=0.5, extentZ=0.5}]
 dyn_fin_l_parent [{shapeType=Box, extentX=6.1737, extentY=1.3492, extentZ=5.6505, center=<2.8875 0 -0.0002>}]
 dyn_fin_r_parent [{shapeType=Box, extentX=4.7581, extentY=1.0532, extentZ=5.774, center=<2.3606 0 0>}]
 dyn_head [{shapeType=Box, extentX=12.6947, extentY=12.6025, extentZ=15.0859, center=<3.6818 0.0024 1.1346>, rotation=<rot 0.0522 -90 -24.5936>}]
 dyn_spine_01 [{shapeType=Box, extentX=3.7643, extentY=3.3766, extentZ=10.0421, center=<-0.3353 0 0>}]
 dyn_spine_01_base [{shapeType=Box, extentX=1, extentY=1, extentZ=1}]
 dyn_spine_02 [{shapeType=Box, extentX=5.4826, extentY=3.7326, extentZ=10.6352, center=<1.6224 0.0002 0>}]
 dyn_spine_03 [{shapeType=Box, extentX=8.2615, extentY=3.7416, extentZ=13.0262, center=<2.9824 0.0001 0>}]
 dyn_spine_04 [{shapeType=Box, extentX=6.0351, extentY=3.5341, extentZ=10.7321, center=<2.4807 0 0>}]
 dyn_tail_base [{shapeType=Box, extentX=0.5, extentY=0.5, extentZ=0.5}]
 dyn_tail_parent [{shapeType=Box, extentX=10.5037, extentY=2.5911, extentZ=15.5288, center=<5.1477 0 0>}]
}
constraints [
 {bone1Name=dyn_tail_parent, bone2Name=dyn_tail_base, swing1LimitDegrees=30, swing2LimitDegrees=5, twistLimitDegrees=3}
 {bone1Name=dyn_spine_02, bone2Name=dyn_spine_01, swing1LimitDegrees=2, swing2LimitDegrees=2, twistLimitDegrees=2}
 {bone1Name=dyn_spine_03, bone2Name=dyn_spine_02, swing1LimitDegrees=2, swing2LimitDegrees=6, twistLimitDegrees=2}
 {bone1Name=dyn_spine_04, bone2Name=dyn_spine_03, swing1LimitDegrees=2, swing2LimitDegrees=6, twistLimitDegrees=2}
 {bone1Name=dyn_spine_01, bone2Name=Melee_TwoHanded_Handle, swing1LimitDegrees=2, swing2LimitDegrees=2, twistLimitDegrees=2}
 {bone1Name=dyn_fin_r_parent, bone2Name=dyn_spine_04, swing1LimitDegrees=15, swing2LimitDegrees=5, twistLimitDegrees=3}
 {bone1Name=dyn_fin_l_parent, bone2Name=dyn_spine_04, swing1LimitDegrees=15, swing2LimitDegrees=5, twistLimitDegrees=3}
 {bone1Name=dyn_head, bone2Name=dyn_spine_04, swing1LimitDegrees=2, swing2LimitDegrees=3, twistLimitDegrees=2}
]
