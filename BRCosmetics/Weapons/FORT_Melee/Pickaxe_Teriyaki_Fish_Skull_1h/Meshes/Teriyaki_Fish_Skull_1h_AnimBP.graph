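# Animation Blueprint for the one-handed Teriyaki Fish Skull pickaxe, adding secondary physics motion to the weapon mesh.
# On initialize it compares the mesh component's name to WeaponMeshOffhand, clearing offhand_weapon on a match and setting it otherwise.
# That flag enables an additive 180-degree yaw on b_AR_Root.
# The pose then feeds a RigidBody node simulating in that bone's space, with damped component motion and overridden gravity.
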AnimBlueprint /BRCosmetics/Weapons/FORT_Melee/Pickaxe_Teriyaki_Fish_Skull_1h/Meshes/Teriyaki_Fish_Skull_1h_AnimBP
target_skeleton=&/Teriyaki_Fish_Skull_1h_Skeleton
graphs [
 AnimGraph {nodes [
   AnimGraphNode_SubInput_0 = Input Pose()
   n1 = "Transform (Modify) Bone - Bone: b_AR_Root"!(
    ComponentPose=Local To Component(LocalPose=@AnimGraphNode_SubInput_0)
    bAlphaBoolEnabled:bool=Get offhand_weapon() -> (offhand_weapon:bool)
    Translation="0,0,0"
    Rotation="0,180,0"
    Scale="1,1,1"
   ) -> (Pose) {
    alphaInputType=Bool
    rotationMode=BMM_Additive
    rotationSpace=BCS_BoneSpace
    translationSpace=BCS_BoneSpace
    alphaBoolBlend {blendOption=Linear}
    boneToModify {boneName=b_AR_Root}
    rotation {yaw=180}
   }
   n2 = RigidBody(ComponentPose=@n1) {
    bClampLinearTranslationLimitToRefPose=true
    bOverrideWorldGravity=true
    overridePhysicsAsset=&/Teriyaki_Fish_Skull_1h_Physics
    simulationSpace=BaseBoneSpace
    alphaBoolBlend {blendOption=Linear}
    baseBoneRef {boneName=b_AR_Root}
    componentAppliedLinearAccClamp=<3000 3000 3000>
    componentLinearAccScale=<0.15 0.15 0.15>
    componentLinearVelScale=<0.15 0.15 0.15>
    overrideWorldGravity {z=-980}
   }
   Output Pose(Result=Component To Local(ComponentPose=@n2)) {layerGroup=None}
  ]}
 EventGraph {nodes [
   Event BlueprintInitializeAnimation()
   > n1 = Set Skel_Mesh(Skel_Mesh=GetPartSkeletalMeshComponent())
   > n2 = Branch(Condition="Equal Exactly (String)"(A="To String (Object)"!(InObj:Object Reference=@n1) -> (ReturnValue:String), B=WeaponMeshOffhand))
   > Set offhand_weapon(offhand_weapon:bool=false) -> (Output_Get:bool)
   Set offhand_weapon(execute=@n2.else, offhand_weapon=true)
  ]}
 NewFunction_0 {nodes [NewFunction_0() {is=title K2Node_FunctionEntry}]}
]
variables [
 offhand_weapon {type=boolean, default=true}
 Skel_Mesh {type=SkeletalMeshComponent}
]
